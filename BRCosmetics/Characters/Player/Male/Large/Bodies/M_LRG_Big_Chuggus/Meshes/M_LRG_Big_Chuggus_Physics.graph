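# Physics asset for the Big Chuggus large male player body, covering its dangling belt piece and two "nade" attachments, likely for secondary motion.
# Each dynamic bone is constrained with angular limits to its own tiny-capsule driver or parent bone, which links to the pelvis with free swing and twist.
# Capsules on the pelvis and left thigh provide body volumes, while boxes and capsules shape the belt and attachments.
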PhysicsAsset /BRCosmetics/Characters/Player/Male/Large/Bodies/M_LRG_Big_Chuggus/Meshes/M_LRG_Big_Chuggus_Physics
bodies {
 dyn_belt_01 [{shapeType=Box, extentX=1.6065, extentY=14.0178, extentZ=6.9738, center=<3.4335 -0.2332 -0.1523>, rotation=<rot 0.3623 -90 -6.5596>}]
 dyn_driver_nadeA [{shapeType=Capsule, length=0.505, radius=0.505, rotation=<rot 0 -90 0>}]
 dyn_driver_nadeB [{shapeType=Capsule, length=0.505, radius=0.505, rotation=<rot 0 -90 0>}]
 dyn_nadeA_01 [{shapeType=Capsule, length=16.1041, radius=4.7984, center=<0 -5 0>, rotation=<rot 0 -90 -90>}]
 dyn_nadeB_01 [
  {shapeType=Capsule, length=10.3541, radius=3.252, center=<2.3661 -4.9909 -0.0002>, rotation=<rot 0 -90 -90>}
  {shapeType=Box, extentX=3, extentY=14.5, extentZ=24.5, center=<10.0361 -6.8245 0.0002>}
 ]
 dyn_parent_belt [{shapeType=Capsule, length=0.505, radius=0.505, rotation=<rot 0 -90 0>}]
 pelvis [{shapeType=Capsule, length=17.5, radius=10.25}]
 thigh_l [{shapeType=Capsule, length=33, radius=10.25, center=<-12.6456 -0.0007 3.935>, rotation=<rot 84.8982 -179.9999 -179.9999>}]
}
constraints [
 {bone1Name=dyn_nadeB_01, bone2Name=dyn_driver_nadeB, swing1LimitDegrees=43.8571, swing2LimitDegrees=25, twistLimitDegrees=25}
 {bone1Name=dyn_nadeA_01, bone2Name=dyn_driver_nadeA, swing1LimitDegrees=35, swing2LimitDegrees=21.5714, twistLimitDegrees=25}
 {
  bone1Name=dyn_driver_nadeA
  bone2Name=pelvis
  swing1LimitDegrees=45
  swing1Motion=Free
  swing2LimitDegrees=45
  swing2Motion=Free
  twistLimitDegrees=45
  twistMotion=Free
 }
 {
  bone1Name=dyn_driver_nadeB
  bone2Name=pelvis
  swing1LimitDegrees=45
  swing1Motion=Free
  swing2LimitDegrees=45
  swing2Motion=Free
  twistLimitDegrees=45
  twistMotion=Free
 }
 {
  bone1Name=thigh_l
  bone2Name=pelvis
  swing1LimitDegrees=45
  swing1Motion=Free
  swing2LimitDegrees=45
  swing2Motion=Free
  twistLimitDegrees=45
  twistMotion=Free
 }
 {
  bone1Name=dyn_parent_belt
  bone2Name=pelvis
  swing1LimitDegrees=45
  swing1Motion=Free
  swing2LimitDegrees=45
  swing2Motion=Free
  twistLimitDegrees=45
  twistMotion=Free
 }
 {bone1Name=dyn_belt_01, bone2Name=dyn_parent_belt, swing1LimitDegrees=17.5714, swing2LimitDegrees=31.2857, twistLimitDegrees=5}
]
