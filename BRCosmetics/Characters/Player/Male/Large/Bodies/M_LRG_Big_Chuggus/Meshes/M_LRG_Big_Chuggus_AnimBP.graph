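# Animation Blueprint for the Big Chuggus body, adding secondary motion to shoulder armor and grenade bones over the attached parent's copied pose.
# Crouch-weighted offsets shift grenade bones, then per-arm Pose Drivers pose shoulder armor using separate grounded and skydiving chains.
# A right-shoulder constraint and arm-up offset feed a RigidBody node, whose alpha drops while skydiving.
# Two grenade AnimDynamics nodes exist but are not connected to the output.
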AnimBlueprint /BRCosmetics/Characters/Player/Male/Large/Bodies/M_LRG_Big_Chuggus/Meshes/M_LRG_Big_Chuggus_AnimBP
target_skeleton=&/M_LRG_Big_Chuggus_Skeleton
graphs [
 AnimGraph {nodes [
   n1 = Anim Dynamics - Bone: dyn_driver_nadeA(Alpha=0.7, ExternalForce="0,0,0") {
    angularDampingOverride=1
    angularSpringConstant=1500
    bAngularSpring=true
    bGravityOverrideInSimSpace=true
    bOverrideAngularDamping=true
    gravityScale=0.32
    simulationSpace=RootRelative
    alphaBoolBlend {blendOption=Linear}
    boundBone {boneName=dyn_driver_nadeA}
    componentAppliedLinearAccClamp=<10000 10000 10000>
    componentLinearAccScale=<1 1 0.5>
    componentLinearVelScale=<1 1 0.5>
    physicsBodyDefinitions [{
      sphereCollisionRadius=0
      boundBone {boneName=dyn_driver_nadeA}
      boxExtents=<15 25 25>
      constraintSetup {angularLimitsMax=<5 25 0>, angularLimitsMin=<-5 -25 -40>, angularTarget=<0 0 0>}
      localJointOffset=<10 0 0>
     }]
   }
   Anim Dynamics - Bone: dyn_driver_nadeB(ComponentPose=@n1, Alpha=0.85, ExternalForce="0,0,0") {
    angularDampingOverride=1
    angularSpringConstant=1500
    bAngularSpring=true
    bGravityOverrideInSimSpace=true
    bOverrideAngularDamping=true
    gravityScale=0.35
    simulationSpace=RootRelative
    alphaBoolBlend {blendOption=Linear}
    boundBone {boneName=dyn_driver_nadeB}
    componentAppliedLinearAccClamp=<3000 3000 3000>
    componentLinearAccScale=<0.5 5 0.5>
    componentLinearVelScale=<0.5 0.5 0.5>
    physicsBodyDefinitions [{
      sphereCollisionRadius=0
      boundBone {boneName=dyn_driver_nadeB}
      boxExtents=<10 20 20>
      constraintSetup {angularLimitsMax=<5 25 0>, angularLimitsMin=<-5 -25 -40>, angularTarget=<0 0 0>}
      localJointOffset=<10 0 0>
     }]
   }
   n2 = "Constraint - dyn_shoulderarmor_master_r (PPPPPP)"!(
    ComponentPose=Local To Component(LocalPose=Use cached pose 'PoseDriver_Cache'())
    Alpha=1
    ConstraintWeights_0=0.0839
    ConstraintWeights_1:float=0.3005
    ConstraintWeights_2:float=0.164
    ConstraintWeights_3:float=0.041
    ConstraintWeights_4:float=0.014
    ConstraintWeights_5:float=0.3966
   ) -> (Pose) {
    alphaBoolBlend {blendOption=Linear}
    boneToModify {boneName=dyn_shoulderarmor_master_r}
    constraintSetup [
     {targetBone {boneName=upperarm_r}}
     {targetBone {boneName=upperarm_twist_01_r}}
     {targetBone {boneName=upperarm_twist_02_r}}
     {targetBone {boneName=spine_03}}
     {targetBone {boneName=spine_05}}
     {targetBone {boneName=clavicle_r}}
    ]
    constraintWeights [1,1,1,1,1,1]
   }
   n3 = "Transform (Modify) Bone - Bone: dyn_shoulderarmor_offset_01_r"(ComponentPose=@n2, Alpha=Get ArmUpR() -> (ArmUpR:float), Translation="-0.300,1.076,0.555", Rotation="-2.772,-11.759,-69.427", Scale="1,1,1") {
    rotationMode=BMM_Additive
    rotationSpace=BCS_BoneSpace
    translationMode=BMM_Additive
    translationSpace=BCS_BoneSpace
    alphaBoolBlend {blendOption=Linear}
    alphaScaleBiasClamp {bInterpResult=true, interpSpeedDecreasing=30, interpSpeedIncreasing=30}
    boneToModify {boneName=dyn_shoulderarmor_offset_01_r}
    rotation=<rot -2.772 -11.7594 -69.4271>
    translation=<-0.3 1.076 0.555>
   }
   n4 = RigidBody(ComponentPose=@n3, Alpha=Get isSkyDivingFloat() -> (isSkyDivingFloat:double)) {
    bClampLinearTranslationLimitToRefPose=true
    bOverrideWorldGravity=true
    overridePhysicsAsset=&/M_LRG_Big_Chuggus_Physics
    simulationSpace=BaseBoneSpace
    alphaBoolBlend {blendOption=Linear}
    baseBoneRef {boneName=root}
    componentAppliedLinearAccClamp=<800 800 1100>
    componentLinearAccScale=<0.5 0.5 0.35>
    componentLinearVelScale=<5 5 0.35>
    overrideWorldGravity {z=-980}
   }
   Output Pose(Result=Component To Local(ComponentPose=@n4)) {layerGroup=None}
   n5 = Pose Driver - Source: clavicle_l!(SourcePose:Pose Link Structure=Use cached pose 'AnimCache'()) -> (Pose:Pose Link Structure) {
    is=title AnimGraphNode_PoseDriver
    poseAsset=&/M_LRG_Big_Chuggus_Poses_PoseAsset
    onlyDriveBones [{boneName=dyn_shoulderarmor_master_l}]
    poseTargets [
     {
      bApplyCustomCurve=false
      bIsHidden=false
      distanceMethod=DefaultMethod
      drivenName=Grounded_Ref
      functionType=DefaultFunction
      targetScale=0.3136
      boneTransforms [
       {targetRotation=<rot -85.1913 -138.5624 143.0618>, targetTranslation=<2.5856 -0.2866 -4.3777>}
       {targetRotation=<rot -31.2301 -1.0781 3.3086>, targetTranslation=<18.2264 0.1342 -0.001>}
      ]
      customCurve {defaultValue=3.4028234663852886e+38, postInfinityExtrap=RCCE_Constant, preInfinityExtrap=RCCE_Constant, keys []}
      targetRotation=<rot 0 0 0>
     }
     {
      bApplyCustomCurve=false
      bIsHidden=false
      distanceMethod=DefaultMethod
      drivenName=Pose_1
      functionType=DefaultFunction
      targetScale=0.6735
      boneTransforms [
       {targetRotation=<rot -71.8075 -7.0774 -33.3057>, targetTranslation=<2.5856 -0.2866 -4.3777>}
       {targetRotation=<rot -4.477 -26.2007 -28.1668>, targetTranslation=<18.2264 0.1342 -0.001>}
      ]
      customCurve {defaultValue=3.4028234663852886e+38, postInfinityExtrap=RCCE_Constant, preInfinityExtrap=RCCE_Constant, keys []}
      targetRotation=<rot 0 0 0>
     }
     {
      bApplyCustomCurve=false
      bIsHidden=false
      distanceMethod=DefaultMethod
      drivenName=Pose_2
      functionType=DefaultFunction
      targetScale=0.6495
      boneTransforms [
       {targetRotation=<rot -75.4618 19.5667 -12.8774>, targetTranslation=<2.5856 -0.2866 -4.3777>}
       {targetRotation=<rot -60.9919 67.2969 -36.8144>, targetTranslation=<18.2264 0.1342 -0.001>}
      ]
      customCurve {defaultValue=3.4028234663852886e+38, postInfinityExtrap=RCCE_Constant, preInfinityExtrap=RCCE_Constant, keys []}
      targetRotation=<rot 0 0 0>
     }
     {
      bApplyCustomCurve=false
      bIsHidden=false
      distanceMethod=DefaultMethod
      drivenName=Pose_3
      functionType=DefaultFunction
      targetScale=0.45
      boneTransforms [
       {targetRotation=<rot -74.7073 55.0349 -49.5566>, targetTranslation=<2.5856 -0.2866 -4.3777>}
       {targetRotation=<rot -30.15 -7.1505 26.6126>, targetTranslation=<18.2264 0.1342 -0.001>}
      ]
      customCurve {defaultValue=3.4028234663852886e+38, postInfinityExtrap=RCCE_Constant, preInfinityExtrap=RCCE_Constant, keys []}
      targetRotation=<rot 0 0 0>
     }
     {
      bApplyCustomCurve=false
      bIsHidden=false
      distanceMethod=DefaultMethod
      drivenName=Pose_4
      functionType=DefaultFunction
      targetScale=0.3136
      boneTransforms [
       {targetRotation=<rot -81.7412 -135.4939 140.2622>, targetTranslation=<2.5856 -0.2866 -4.3777>}
       {targetRotation=<rot -43.3901 -9.8542 2.3218>, targetTranslation=<18.2264 0.1342 -0.001>}
      ]
      customCurve {defaultValue=3.4028234663852886e+38, postInfinityExtrap=RCCE_Constant, preInfinityExtrap=RCCE_Constant, keys []}
      targetRotation=<rot 0 0 0>
     }
     {
      bApplyCustomCurve=false
      bIsHidden=false
      distanceMethod=DefaultMethod
      drivenName=Pose_5
      functionType=DefaultFunction
      targetScale=0.6495
      boneTransforms [
       {targetRotation=<rot -80.8623 39.04 -38.1229>, targetTranslation=<2.5856 -0.2866 -4.3777>}
       {targetRotation=<rot -32.5004 58.0615 -81.999>, targetTranslation=<18.2264 0.1342 -0.001>}
      ]
      customCurve {defaultValue=3.4028234663852886e+38, postInfinityExtrap=RCCE_Constant, preInfinityExtrap=RCCE_Constant, keys []}
      targetRotation=<rot 0 0 0>
     }
     {
      bApplyCustomCurve=false
      bIsHidden=false
      distanceMethod=DefaultMethod
      drivenName=Pose_6
      functionType=DefaultFunction
      targetScale=0.7692
      boneTransforms [
       {targetRotation=<rot -47.5697 14.692 -0.434>, targetTranslation=<2.5856 -0.2866 -4.3777>}
       {targetRotation=<rot -8.3495 19.7449 -78.0733>, targetTranslation=<18.2264 0.1342 -0.001>}
      ]
      customCurve {defaultValue=3.4028234663852886e+38, postInfinityExtrap=RCCE_Constant, preInfinityExtrap=RCCE_Constant, keys []}
      targetRotation=<rot 0 0 0>
     }
     {
      bApplyCustomCurve=false
      bIsHidden=false
      distanceMethod=DefaultMethod
      drivenName=Pose_9
      functionType=DefaultFunction
      targetScale=1
      boneTransforms [
       {targetRotation=<rot -48.3561 10.1188 -2.4428>, targetTranslation=<2.5856 -0.2866 -4.3777>}
       {targetRotation=<rot 12.1943 48.5446 -100.0629>, targetTranslation=<18.2264 0.1342 -0.001>}
      ]
      customCurve {defaultValue=3.4028234663852886e+38, postInfinityExtrap=RCCE_Constant, preInfinityExtrap=RCCE_Constant, keys []}
      targetRotation=<rot 0 0 0>
     }
     {
      bApplyCustomCurve=false
      bIsHidden=false
      distanceMethod=DefaultMethod
      drivenName=Pose_10
      functionType=DefaultFunction
      targetScale=1
      boneTransforms [
       {targetRotation=<rot -79.2873 99.6394 -110.9318>, targetTranslation=<2.5856 -0.2866 -4.3777>}
       {targetRotation=<rot -9.1378 99.5965 -17.0808>, targetTranslation=<18.2264 0.1342 -0.001>}
      ]
      customCurve {defaultValue=3.4028234663852886e+38, postInfinityExtrap=RCCE_Constant, preInfinityExtrap=RCCE_Constant, keys []}
      targetRotation=<rot 0 0 0>
     }
     {
      bApplyCustomCurve=false
      bIsHidden=false
      distanceMethod=DefaultMethod
      drivenName=Pose_11
      functionType=DefaultFunction
      targetScale=0.6735
      boneTransforms [
       {targetRotation=<rot -61.5846 27.1484 -22.3371>, targetTranslation=<2.5856 -0.2866 -4.3777>}
       {targetRotation=<rot 18.8241 -14.1266 -74.4435>, targetTranslation=<18.2264 0.1342 -0.001>}
      ]
      customCurve {defaultValue=3.4028234663852886e+38, postInfinityExtrap=RCCE_Constant, preInfinityExtrap=RCCE_Constant, keys []}
      targetRotation=<rot 0 0 0>
     }
     {
      bApplyCustomCurve=false
      bIsHidden=false
      distanceMethod=SwingAngle
      drivenName=Pose_12
      functionType=Gaussian
      targetScale=0.9262
      boneTransforms [
       {targetRotation=<rot -58.4846 -9.5373 2.6759>, targetTranslation=<2.5856 -0.2866 -4.3777>}
       {targetRotation=<rot -41.1313 120.896 -42.9659>, targetTranslation=<18.2264 0.1342 -0.001>}
      ]
      customCurve {defaultValue=3.4028234663852886e+38, postInfinityExtrap=RCCE_Constant, preInfinityExtrap=RCCE_Constant, keys []}
      targetRotation=<rot 0 0 0>
     }
    ]
    rBFParams {radius=40}
    sourceBones [{boneName=clavicle_l}, {boneName=upperarm_l}]
   }
   n6 = Pose Driver - Source: clavicle_r!(SourcePose:Pose Link Structure=@n5) -> (Pose:Pose Link Structure) {
    is=title AnimGraphNode_PoseDriver
    poseAsset=&M_LRG_Big_Chuggus_Poses_PoseAsset
    onlyDriveBones [{boneName=dyn_shoulderarmor_master_r}, {boneName=dyn_shoulderarmor_offset_01_r}]
    poseTargets [
     {
      bApplyCustomCurve=false
      bIsHidden=false
      distanceMethod=DefaultMethod
      drivenName=Grounded_Ref
      functionType=DefaultFunction
      targetScale=0.862
      boneTransforms [
       {targetRotation=<rot -85.1915 41.4375 143.0619>, targetTranslation=<4.3575 -0.1466 2.3223>}
       {targetRotation=<rot -37.7574 0.1509 2.6875>, targetTranslation=<-15.5111 0 0>}
      ]
      customCurve {defaultValue=3.4028234663852886e+38, postInfinityExtrap=RCCE_Constant, preInfinityExtrap=RCCE_Constant, keys []}
      targetRotation=<rot 0 0 0>
     }
     {
      bApplyCustomCurve=false
      bIsHidden=false
      distanceMethod=DefaultMethod
      drivenName=Pose_1
      functionType=DefaultFunction
      targetScale=0.6847
      boneTransforms [
       {targetRotation=<rot -64.2747 -120.6598 -83.241>, targetTranslation=<4.3575 -0.1466 2.3223>}
       {targetRotation=<rot -23.3709 86.7554 -36.3331>, targetTranslation=<-15.5111 0 0>}
      ]
      customCurve {defaultValue=3.4028234663852886e+38, postInfinityExtrap=RCCE_Constant, preInfinityExtrap=RCCE_Constant, keys []}
      targetRotation=<rot 0 0 0>
     }
     {
      bApplyCustomCurve=false
      bIsHidden=false
      distanceMethod=DefaultMethod
      drivenName=Pose_2
      functionType=DefaultFunction
      targetScale=0.3773
      boneTransforms [
       {targetRotation=<rot -69.838 165.5965 22.6769>, targetTranslation=<4.3575 -0.1466 2.3223>}
       {targetRotation=<rot -75.1088 43.0245 -78.2219>, targetTranslation=<-15.5111 0 0>}
      ]
      customCurve {defaultValue=3.4028234663852886e+38, postInfinityExtrap=RCCE_Constant, preInfinityExtrap=RCCE_Constant, keys []}
      targetRotation=<rot 0 0 0>
     }
     {
      bApplyCustomCurve=false
      bIsHidden=false
      distanceMethod=DefaultMethod
      drivenName=Pose_3
      functionType=DefaultFunction
      targetScale=0.3773
      boneTransforms [
       {targetRotation=<rot -73.4134 -176.9703 4.3592>, targetTranslation=<4.3575 -0.1466 2.3223>}
       {targetRotation=<rot -75.5656 105.7208 -93.8077>, targetTranslation=<-15.5111 0 0>}
      ]
      customCurve {defaultValue=3.4028234663852886e+38, postInfinityExtrap=RCCE_Constant, preInfinityExtrap=RCCE_Constant, keys []}
      targetRotation=<rot 0 0 0>
     }
     {
      bApplyCustomCurve=false
      bIsHidden=false
      distanceMethod=DefaultMethod
      drivenName=Pose_4
      functionType=DefaultFunction
      targetScale=0.4826
      boneTransforms [
       {targetRotation=<rot -87.7993 76.2682 109.0758>, targetTranslation=<4.3575 -0.1466 2.3223>}
       {targetRotation=<rot -69.7589 51.1037 -35.2693>, targetTranslation=<-15.5111 0 0>}
      ]
      customCurve {defaultValue=3.4028234663852886e+38, postInfinityExtrap=RCCE_Constant, preInfinityExtrap=RCCE_Constant, keys []}
      targetRotation=<rot 0 0 0>
     }
     {
      bApplyCustomCurve=false
      bIsHidden=false
      distanceMethod=DefaultMethod
      drivenName=Pose_5
      functionType=DefaultFunction
      targetScale=0.3608
      boneTransforms [
       {targetRotation=<rot -76.2754 172.1506 9.5506>, targetTranslation=<4.3575 -0.1466 2.3223>}
       {targetRotation=<rot -45.0185 104.3596 -56.1379>, targetTranslation=<-15.5111 0 0>}
      ]
      customCurve {defaultValue=3.4028234663852886e+38, postInfinityExtrap=RCCE_Constant, preInfinityExtrap=RCCE_Constant, keys []}
      targetRotation=<rot 0 0 0>
     }
     {
      bApplyCustomCurve=false
      bIsHidden=false
      distanceMethod=DefaultMethod
      drivenName=Pose_6
      functionType=DefaultFunction
      targetScale=0.3608
      boneTransforms [
       {targetRotation=<rot -74.8686 -132.7057 -46.9771>, targetTranslation=<4.3575 -0.1466 2.3223>}
       {targetRotation=<rot -39.6825 113.2505 -70.71>, targetTranslation=<-15.5111 0 0>}
      ]
      customCurve {defaultValue=3.4028234663852886e+38, postInfinityExtrap=RCCE_Constant, preInfinityExtrap=RCCE_Constant, keys []}
      targetRotation=<rot 0 0 0>
     }
     {
      bApplyCustomCurve=false
      bIsHidden=false
      distanceMethod=DefaultMethod
      drivenName=Pose_7
      functionType=DefaultFunction
      targetScale=0.4435
      boneTransforms [
       {targetRotation=<rot -57.8419 -161.5812 -18.7948>, targetTranslation=<4.3575 -0.1466 2.3223>}
       {targetRotation=<rot -66.8684 120.4314 -71.1722>, targetTranslation=<-15.5111 0 0>}
      ]
      customCurve {defaultValue=3.4028234663852886e+38, postInfinityExtrap=RCCE_Constant, preInfinityExtrap=RCCE_Constant, keys []}
      targetRotation=<rot 0 0 0>
     }
     {
      bApplyCustomCurve=false
      bIsHidden=false
      distanceMethod=DefaultMethod
      drivenName=Pose_8
      functionType=DefaultFunction
      targetScale=0.9584
      boneTransforms [
       {targetRotation=<rot -53.6465 -152.7178 -29.1383>, targetTranslation=<4.3575 -0.1466 2.3223>}
       {targetRotation=<rot -17.8084 26.8504 -97.253>, targetTranslation=<-15.5111 0 0>}
      ]
      customCurve {defaultValue=3.4028234663852886e+38, postInfinityExtrap=RCCE_Constant, preInfinityExtrap=RCCE_Constant, keys []}
      targetRotation=<rot 0 0 0>
     }
     {
      bApplyCustomCurve=false
      bIsHidden=false
      distanceMethod=DefaultMethod
      drivenName=Pose_9
      functionType=DefaultFunction
      targetScale=0.7854
      boneTransforms [
       {targetRotation=<rot -68.9933 -127.6298 -53.9107>, targetTranslation=<4.3575 -0.1466 2.3223>}
       {targetRotation=<rot 33.2732 89.3676 -89.1674>, targetTranslation=<-15.5111 0 0>}
      ]
      customCurve {defaultValue=3.4028234663852886e+38, postInfinityExtrap=RCCE_Constant, preInfinityExtrap=RCCE_Constant, keys []}
      targetRotation=<rot 0 0 0>
     }
     {
      bApplyCustomCurve=false
      bIsHidden=false
      distanceMethod=DefaultMethod
      drivenName=Pose_10
      functionType=DefaultFunction
      targetScale=0.862
      boneTransforms [
       {targetRotation=<rot -83.622 69.4344 115.2173>, targetTranslation=<4.3575 -0.1466 2.3223>}
       {targetRotation=<rot -11.4935 30.1498 -76.7725>, targetTranslation=<-15.5111 0 0>}
      ]
      customCurve {defaultValue=3.4028234663852886e+38, postInfinityExtrap=RCCE_Constant, preInfinityExtrap=RCCE_Constant, keys []}
      targetRotation=<rot 0 0 0>
     }
     {
      bApplyCustomCurve=false
      bIsHidden=false
      distanceMethod=DefaultMethod
      drivenName=Pose_11
      functionType=DefaultFunction
      targetScale=1
      boneTransforms [
       {targetRotation=<rot -66.8403 -177.8105 2.9442>, targetTranslation=<4.3575 -0.1466 2.3223>}
       {targetRotation=<rot 18.2092 9.4898 -75.6901>, targetTranslation=<-15.5111 0 0>}
      ]
      customCurve {defaultValue=3.4028234663852886e+38, postInfinityExtrap=RCCE_Constant, preInfinityExtrap=RCCE_Constant, keys []}
      targetRotation=<rot 0 0 0>
     }
     {
      bApplyCustomCurve=false
      bIsHidden=false
      distanceMethod=DefaultMethod
      drivenName=Pose_12
      functionType=DefaultFunction
      targetScale=0.724
      boneTransforms [
       {targetRotation=<rot -47.9536 -145.5199 -31.7961>, targetTranslation=<4.3575 -0.1466 2.3223>}
       {targetRotation=<rot -20.6591 112.0123 -37.7437>, targetTranslation=<-15.5111 0 0>}
      ]
      customCurve {defaultValue=3.4028234663852886e+38, postInfinityExtrap=RCCE_Constant, preInfinityExtrap=RCCE_Constant, keys []}
      targetRotation=<rot 0 0 0>
     }
     {
      bApplyCustomCurve=false
      bIsHidden=false
      distanceMethod=DefaultMethod
      drivenName=Pose_13
      functionType=DefaultFunction
      targetScale=0.7854
      boneTransforms [
       {targetRotation=<rot -86.6669 -153.8735 -21.7279>, targetTranslation=<4.3575 -0.1466 2.3223>}
       {targetRotation=<rot 61.9226 97.9326 -80.7543>, targetTranslation=<-15.5111 0 0>}
      ]
      customCurve {defaultValue=3.4028234663852886e+38, postInfinityExtrap=RCCE_Constant, preInfinityExtrap=RCCE_Constant, keys []}
      targetRotation=<rot 0 0 0>
     }
    ]
    rBFParams {radius=50}
    sourceBones [{boneName=clavicle_r}, {boneName=upperarm_r}]
   }
   n7 = Pose Driver - Source: upperarm_l!(SourcePose:Pose Link Structure=Use cached pose 'AnimCache'()) -> (Pose:Pose Link Structure) {
    is=title AnimGraphNode_PoseDriver
    poseAsset=&M_LRG_Big_Chuggus_Poses_PoseAsset
    onlyDriveBones [{boneName=dyn_shoulderarmor_master_l}]
    poseTargets [
     {
      bApplyCustomCurve=false
      bIsHidden=false
      distanceMethod=DefaultMethod
      drivenName=Skydive_Ref
      functionType=DefaultFunction
      targetScale=0.7213
      boneTransforms [{targetRotation=<rot -58.2076 -4.3462 7.126>, targetTranslation=<18.2264 0.1342 -0.001>}]
      customCurve {defaultValue=3.4028234663852886e+38, postInfinityExtrap=RCCE_Constant, preInfinityExtrap=RCCE_Constant, keys []}
      targetRotation=<rot 0 0 0>
     }
     {
      bApplyCustomCurve=false
      bIsHidden=false
      distanceMethod=DefaultMethod
      drivenName=Pose_15
      functionType=DefaultFunction
      targetScale=1
      boneTransforms [{targetRotation=<rot -24.9516 -15.213 -65.9431>, targetTranslation=<18.2264 0.1342 -0.001>}]
      customCurve {defaultValue=3.4028234663852886e+38, postInfinityExtrap=RCCE_Constant, preInfinityExtrap=RCCE_Constant, keys []}
      targetRotation=<rot 0 0 0>
     }
     {
      bApplyCustomCurve=false
      bIsHidden=false
      distanceMethod=DefaultMethod
      drivenName=Pose_16
      functionType=DefaultFunction
      targetScale=1
      boneTransforms [{targetRotation=<rot 18.4125 9.339 -71.376>, targetTranslation=<18.2264 0.1342 -0.001>}]
      customCurve {defaultValue=3.4028234663852886e+38, postInfinityExtrap=RCCE_Constant, preInfinityExtrap=RCCE_Constant, keys []}
      targetRotation=<rot 0 0 0>
     }
     {
      bApplyCustomCurve=false
      bIsHidden=false
      distanceMethod=DefaultMethod
      drivenName=Pose_17
      functionType=DefaultFunction
      targetScale=0.7213
      boneTransforms [{targetRotation=<rot -35.6776 -43.5422 32.8522>, targetTranslation=<18.2264 0.1342 -0.001>}]
      customCurve {defaultValue=3.4028234663852886e+38, postInfinityExtrap=RCCE_Constant, preInfinityExtrap=RCCE_Constant, keys []}
      targetRotation=<rot 0 0 0>
     }
    ]
    rBFParams {radius=24.7584}
    sourceBones [{boneName=upperarm_l}]
   }
   n8 = Pose Driver - Source: upperarm_r!(SourcePose:Pose Link Structure=@n7) -> (Pose:Pose Link Structure) {
    is=title AnimGraphNode_PoseDriver
    poseAsset=&M_LRG_Big_Chuggus_Poses_PoseAsset
    onlyDriveBones [{boneName=dyn_shoulderarmor_master_r}, {boneName=dyn_shoulderarmor_offset_01_r}]
    poseTargets [
     {
      bApplyCustomCurve=false
      bIsHidden=false
      distanceMethod=DefaultMethod
      drivenName=Skydive_Ref
      functionType=DefaultFunction
      targetScale=0.6258
      boneTransforms [{targetRotation=<rot -31.2301 -1.0781 3.3086>, targetTranslation=<-18.2264 -0.1342 0.001>}]
      customCurve {defaultValue=3.4028234663852886e+38, postInfinityExtrap=RCCE_Constant, preInfinityExtrap=RCCE_Constant, keys []}
      targetRotation=<rot 0 0 0>
     }
     {
      bApplyCustomCurve=false
      bIsHidden=false
      distanceMethod=DefaultMethod
      drivenName=Pose_15
      functionType=DefaultFunction
      targetScale=1
      boneTransforms [{targetRotation=<rot -31.2301 -1.0781 3.3086>, targetTranslation=<-18.2264 -0.1342 0.001>}]
      customCurve {defaultValue=3.4028234663852886e+38, postInfinityExtrap=RCCE_Constant, preInfinityExtrap=RCCE_Constant, keys []}
      targetRotation=<rot 0 0 0>
     }
     {
      bApplyCustomCurve=false
      bIsHidden=false
      distanceMethod=DefaultMethod
      drivenName=Pose_16
      functionType=DefaultFunction
      targetScale=1
      boneTransforms [{targetRotation=<rot -31.2301 -1.0781 3.3086>, targetTranslation=<-18.2264 -0.1342 0.001>}]
      customCurve {defaultValue=3.4028234663852886e+38, postInfinityExtrap=RCCE_Constant, preInfinityExtrap=RCCE_Constant, keys []}
      targetRotation=<rot 0 0 0>
     }
     {
      bApplyCustomCurve=false
      bIsHidden=false
      distanceMethod=DefaultMethod
      drivenName=Pose_17
      functionType=DefaultFunction
      targetScale=0.6258
      boneTransforms [{targetRotation=<rot -31.2301 -1.0781 3.3086>, targetTranslation=<-18.2264 -0.1342 0.001>}]
      customCurve {defaultValue=3.4028234663852886e+38, postInfinityExtrap=RCCE_Constant, preInfinityExtrap=RCCE_Constant, keys []}
      targetRotation=<rot 0 0 0>
     }
    ]
    rBFParams {radius=0}
    sourceBones [{boneName=upperarm_r}]
   }
   Save cached pose 'PoseDriver_Cache'(Pose=Blend Poses by bool(BlendPose_0=@n8, BlendPose_1=@n6, bActiveValue=Get bIsSkyDiving()) {blendType=Linear}) {cacheName=PoseDriver_Cache}
   n9 = Get CrouchAlpha() -> (CrouchAlpha:float)
   n10 = "Transform (Modify) Bone - Bone: dyn_parent_nadeA"(
    ComponentPose=Local To Component(LocalPose=Copy Pose From Mesh() {bUseAttachedParent=true})
    Alpha=@n9
    Translation="1.112,-4.022,-0.275"
    Rotation="0.000,-15.001,0.000"
    Scale="1,1,1"
   ) {
    rotationMode=BMM_Additive
    rotationSpace=BCS_BoneSpace
    translationMode=BMM_Additive
    translationSpace=BCS_BoneSpace
    alphaBoolBlend {blendOption=Linear}
    alphaScaleBiasClamp {bInterpResult=true, interpSpeedDecreasing=15, interpSpeedIncreasing=15}
    boneToModify {boneName=dyn_parent_nadeA}
    rotation {yaw=-15.001}
    translation=<1.112 -4.022 -0.275>
   }
   n11 = "Transform (Modify) Bone - Bone: dyn_parent_nadeB"(ComponentPose=@n10, Alpha=@n9, Translation="-0.934,-4.159,0.000", Rotation="0.000,-15.001,0.000", Scale="1,1,1") {
    rotationMode=BMM_Additive
    rotationSpace=BCS_BoneSpace
    translationMode=BMM_Additive
    translationSpace=BCS_BoneSpace
    alphaBoolBlend {blendOption=Linear}
    alphaScaleBiasClamp {bInterpResult=true, interpSpeedDecreasing=15, interpSpeedIncreasing=15}
    boneToModify {boneName=dyn_parent_nadeB}
    rotation {yaw=-15.0015}
    translation=<-0.9343 -4.1589 0>
   }
   Save cached pose 'AnimCache'(Pose=Component To Local(ComponentPose=@n11)) {cacheName=AnimCache}
  ]}
 EventGraph {nodes [
   Event BlueprintUpdateAnimation()
   > Sequence()
   > n1 = Branch(Condition=Get bIsSkyDiving())
   > Set isSkyDivingFloat(isSkyDivingFloat:double=0.3) -> (Output_Get:double)
   Set isSkyDivingFloat(execute=@n1.else, isSkyDivingFloat=0.9)
  ]}
]
variables [isSkyDivingFloat {type=number}]
