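# Control Rig for Big Chuggus's head that procedurally adjusts the head, neck and attached tube bones.
# It slerps local rotations from identity to give head_offset most of head's rotation and dampen neck_01.
# Head rotation relative to spine_05 drives smoothed additive offsets on tube and pony bones.
# Tube driver bones then aim along vectors to their target tube bones; an unwired radial clamp and debug draws remain.
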ControlRigBlueprint /BRCosmetics/Characters/Player/Male/Large/Bodies/M_LRG_Big_Chuggus/Meshes/Parts/Male_Large_Big_Chuggus_Head_CtrlRig
graphs [
 RigVMFunctionLibrary {links [], nodes []}
 RigVMModel {nodes [
   n1 = &RigUnit_BeginExecution()
   RigUnit_GetTransform = &RigUnit_GetTransform(Item=(Type=Bone,Name="head"), Space=LocalSpace)
   RigVMFunction_MathQuaternionMake = /Script/RigVM.RigVMFunction_MathQuaternionMake(W=1, X=0, Y=0, Z=0) {is=struct RigVMUnitNode}
   Interpolate = /Script/RigVM.RigVMFunction_MathQuaternionSlerp(T=0.8, B=@RigUnit_GetTransform.Transform.Rotation, A=@RigVMFunction_MathQuaternionMake.Result) {is=struct RigVMUnitNode, template="Interpolate::Execute(in A,in B,in T,out Result)"}
   RigUnit_SetTransform = &RigUnit_SetTransform(
    Item=(Type=Bone,Name="head_offset")
    Space=LocalSpace
    Value=(Rotation=(X=-0.0029,Y=0.7071,Z=0.0029,W=-0.7071),Translation=(X=0,Y=0.9556,Z=152.0342),Scale3D=(X=1,Y=1,Z=1))
    Weight=1
    bInitial=False
    ExecutePin=@n1.ExecutePin
    Value.Rotation=@Interpolate.Result
    Value.Scale3D=@RigUnit_GetTransform.Transform.Scale3D
    Value.Translation=@RigUnit_GetTransform.Transform.Translation
   )
   n2 = &RigUnit_GetTransform(Item=(Type=Bone,Name="neck_01"), Space=LocalSpace)
   > &RigUnit_SetTransform(
    Item=(Type=Bone,Name="neck_01")
    Space=LocalSpace
    Value=(Rotation=(X=-0.0029,Y=0.7071,Z=0.0029,W=-0.7071),Translation=(X=0,Y=0.9556,Z=152.0342),Scale3D=(X=1,Y=1,Z=1))
    Weight=1
    bInitial=False
    Value.Rotation=&RigVMFunction_MathQuaternionSlerp(T=0.3, B=@n2.Transform.Rotation, A=&RigVMFunction_MathQuaternionMake(W=1, X=0, Y=0, Z=0).Result) {template="Interpolate::Execute(in A,in B,in T,out Result)"}.Result
    Value.Scale3D=@n2.Transform.Scale3D
    Value.Translation=@n2.Transform.Translation
   )
   n3 = &RigVMFunction_MathQuaternionToEuler(RotationOrder=XZY, Value=&RigUnit_GetRelativeBoneTransform(Bone=head, Space=spine_05).Transform.Rotation)
   > /Script/ControlRig.RigUnit_ModifyBoneTransforms(
    BoneToModify=((Bone="dyn_tubeM_01_l",Transform=(Rotation=(X=0,Y=0,Z=0,W=1),Translation=(X=0,Y=-8,Z=1),Scale3D=(X=1,Y=1,Z=1))),(Bone="dyn_tubeA_03_l",Transform=(Rotation=(X=0,Y=0,Z=0,W=1),Translation=(X=0,Y=0,Z=0),Scale3D=(X=1,Y=1,Z=1))))
    Mode=AdditiveLocal
    WeightMaximum=1
    WeightMinimum=0
    Weight=/Script/RigVM.RigVMFunction_AccumulateFloatLerp(
     Blend=25
     InitialValue=0
     bIntegrateDeltaTime=true
     TargetValue=&RigVMFunction_MathFloatRemap(SourceMaximum=-50, SourceMinimum=10, TargetMinimum=0, bClamp=true, Value=@n3.Result.X).Result
    ) {is=struct RigVMUnitNode, template="AccumulateLerp::Execute(in TargetValue,in InitialValue,in Blend,in bIntegrateDeltaTime,out Result)"}.Result
   ) {is=struct RigVMUnitNode}
   > /Script/ControlRig.RigUnit_DebugTransformMutable(
    Color=(R=1,G=0,B=0,A=1)
    DebugDrawSettings=(DepthPriority=SDPG_Foreground,Lifetime=-1)
    Mode=Axes
    Scale=10
    Space=dyn_tubeM_01_l
    Thickness=0
    Transform=(Rotation=(X=0,Y=0,Z=0,W=1),Translation=(X=0,Y=0,Z=0),Scale3D=(X=1,Y=1,Z=1))
    WorldOffset=(Rotation=(X=0,Y=0,Z=0,W=1),Translation=(X=0,Y=0,Z=0),Scale3D=(X=1,Y=1,Z=1))
    bEnabled=True
   ) {is=struct RigVMUnitNode}
   > &RigUnit_DebugTransformMutable(
    Color=(R=1,G=0,B=0,A=1)
    DebugDrawSettings=(DepthPriority=SDPG_Foreground,Lifetime=-1)
    Mode=Axes
    Scale=10
    Space=dyn_head_driver_tubeA_l
    Thickness=0
    Transform=(Rotation=(X=0,Y=0,Z=0,W=1),Translation=(X=0,Y=0,Z=0),Scale3D=(X=1,Y=1,Z=1))
    WorldOffset=(Rotation=(X=0,Y=0,Z=0,W=1),Translation=(X=0,Y=0,Z=0),Scale3D=(X=1,Y=1,Z=1))
    bEnabled=True
   )
   n4 = &RigVMFunction_MathQuaternionToEuler(RotationOrder=XZY, Value=&RigUnit_GetRelativeBoneTransform(Bone=head, Space=spine_05).Transform.Rotation)
   > &RigUnit_ModifyBoneTransforms(
    BoneToModify=((Bone="dyn_tubeM_01_r",Transform=(Rotation=(X=0,Y=0,Z=0,W=1),Translation=(X=0,Y=8,Z=-1),Scale3D=(X=1,Y=1,Z=1))))
    Mode=AdditiveLocal
    WeightMaximum=1
    WeightMinimum=0
    Weight=&RigVMFunction_AccumulateFloatLerp(
     Blend=25
     InitialValue=0
     bIntegrateDeltaTime=true
     TargetValue=&RigVMFunction_MathFloatRemap(SourceMaximum=50, SourceMinimum=-10, TargetMinimum=0, bClamp=true, Value=@n4.Result.X).Result
    ) {template="AccumulateLerp::Execute(in TargetValue,in InitialValue,in Blend,in bIntegrateDeltaTime,out Result)"}.Result
   )
   n5 = &RigVMFunction_MathQuaternionToEuler(RotationOrder=XZY, Value=&RigUnit_GetRelativeBoneTransform(Bone=head, Space=spine_05).Transform.Rotation)
   > &RigUnit_ModifyBoneTransforms(
    BoneToModify=((Bone="dyn_parent_pony",Transform=(Rotation=(X=0,Y=0,Z=0,W=1),Translation=(X=5,Y=0.5,Z=-2),Scale3D=(X=1,Y=1,Z=1))))
    Mode=AdditiveLocal
    WeightMaximum=1
    WeightMinimum=0
    Weight=&RigVMFunction_AccumulateFloatLerp(
     Blend=10
     InitialValue=0
     bIntegrateDeltaTime=true
     TargetValue=&RigVMFunction_MathFloatRemap(SourceMaximum=40, SourceMinimum=-8, TargetMinimum=0, bClamp=true, Value=@n5.Result.Z).Result
    ) {template="AccumulateLerp::Execute(in TargetValue,in InitialValue,in Blend,in bIntegrateDeltaTime,out Result)"}.Result
    BoneToModify.0.Transform.Rotation=&RigVMFunction_MathQuaternionFromEuler(Euler=(X=-10,Y=-10,Z=0)).Result
   )
   n6 = &RigVMFunction_MathQuaternionToEuler(RotationOrder=XZY, Value=&RigUnit_GetRelativeBoneTransform(Bone=head, Space=spine_05).Transform.Rotation)
   > &RigUnit_ModifyBoneTransforms(
    BoneToModify="((Bone=\"dyn_tubeC_03_l\",Transform=(Rotation=(X=0,Y=0,Z=0,W=1),Translation=(X=3,Y=-8,Z=-3),Scale3D=(X=1,Y=1,Z=1))),(Bone=\"dyn_tubeC_03_r\",Transform=(Rotation=(X=0,Y=0,Z=0,W=1),Translation=(X=-3,Y=8,Z=3),Scale3D=(X=1,Y=1,Z=1))),(Bone=\"dyn_tubeM_01_l\",Transfo ...(truncated)"
    Mode=AdditiveLocal
    WeightMaximum=1
    WeightMinimum=0
    Weight=&RigVMFunction_AccumulateFloatLerp(
     Blend=25
     InitialValue=0
     bIntegrateDeltaTime=true
     TargetValue=&RigVMFunction_MathFloatRemap(SourceMaximum=-40, SourceMinimum=0, TargetMinimum=0, bClamp=true, Value=@n6.Result.Z).Result
    ) {template="AccumulateLerp::Execute(in TargetValue,in InitialValue,in Blend,in bIntegrateDeltaTime,out Result)"}.Result
   )
   > &RigUnit_DebugTransformMutable(
    Color=(R=1,G=0,B=0,A=1)
    DebugDrawSettings=(DepthPriority=SDPG_Foreground,Lifetime=-1)
    Mode=Axes
    Scale=20
    Space=dyn_driver_tubeB_l
    Thickness=0
    Transform=(Rotation=(X=0,Y=0,Z=0,W=1),Translation=(X=0,Y=0,Z=0),Scale3D=(X=1,Y=1,Z=1))
    WorldOffset=(Rotation=(X=0,Y=0,Z=0,W=1),Translation=(X=0,Y=0,Z=0),Scale3D=(X=1,Y=1,Z=1))
    bEnabled=false
   )
   n7 = &RigUnit_GetBoneTransform(Bone=dyn_tubeC_03_l)
   n8 = /Script/RigVM.RigVMFunction_MathQuaternionGetAxis(Axis=X, Quaternion=@n7.Transform.Rotation) {is=struct RigVMUnitNode, template="Axis::Execute(in Quaternion,in Axis,out Result)"}
   n9 = /Script/RigVM.RigVMFunction_MathVectorAdd(A=(X=8,Y=8,Z=-8), B=(X=0,Y=0,Z=0)) {is=struct RigVMUnitNode, template="Add::Execute(in A,in B,out Result)"}
   n10 = /Script/RigVM.RigVMFunction_MathVectorSub(
    A=@n7.Transform.Translation
    B=/Script/RigVM.RigVMFunction_MathTransformTransformVector(
     Transform=&RigUnit_GetBoneTransform(Bone=dyn_driver_tubeB_l).Transform
     Location=/Script/RigVM.RigVMFunction_MathVectorSelectBool(Condition=False, IfFalse=(X=0,Y=0,Z=0), IfTrue=@n9.Result) {is=struct RigVMUnitNode}.Result
    ) {is=struct RigVMUnitNode}.Result
   ) {is=struct RigVMUnitNode, template="Subtract::Execute(in A,in B,out Result)"}
   > /Script/ControlRig.RigUnit_SetBoneRotation(
    Bone=dyn_driver_tubeB_l
    Space=GlobalSpace
    Weight=1
    bPropagateToChildren=True
    Rotation=/Script/RigVM.RigVMFunction_MathQuaternionMul(
     B=@n7.Transform.Rotation
     A=/Script/RigVM.RigVMFunction_MathQuaternionFromTwoVectors(A=@n8.Result, B=@n10.Result) {is=struct RigVMUnitNode, template="FromTwoVectors::Execute(in A,in B,out Result)"}.Result
    ) {is=struct RigVMUnitNode, template="Multiply::Execute(in A,in B,out Result)"}.Result
   ) {is=struct RigVMUnitNode}
   > &RigUnit_DebugTransformMutable(
    Color=(R=1,G=0,B=0,A=1)
    DebugDrawSettings=(DepthPriority=SDPG_Foreground,Lifetime=-1)
    Mode=Axes
    Scale=50
    Space=None
    Thickness=0
    WorldOffset=(Rotation=(X=0,Y=0,Z=0,W=1),Translation=(X=0,Y=0,Z=0),Scale3D=(X=1,Y=1,Z=1))
    bEnabled=false
    Transform=&RigUnit_GetBoneTransform(Bone=dyn_driver_tubeB_l).Transform
   )
   > &RigUnit_DebugTransformMutable(
    Color=(R=1,G=0,B=0,A=1)
    DebugDrawSettings=(DepthPriority=SDPG_Foreground,Lifetime=-1)
    Mode=Axes
    Scale=20
    Space=dyn_head_driver_tubeA_l
    Thickness=0
    Transform=(Rotation=(X=0,Y=0,Z=0,W=1),Translation=(X=0,Y=0,Z=0),Scale3D=(X=1,Y=1,Z=1))
    WorldOffset=(Rotation=(X=0,Y=0,Z=0,W=1),Translation=(X=0,Y=0,Z=0),Scale3D=(X=1,Y=1,Z=1))
    bEnabled=false
   )
   n11 = &RigUnit_GetBoneTransform(Bone=dyn_tubeM_01_l)
   n12 = &RigVMFunction_MathQuaternionGetAxis(Axis=X, Quaternion=@n11.Transform.Rotation) {template="Axis::Execute(in Quaternion,in Axis,out Result)"}
   n13 = &RigVMFunction_MathVectorAdd(A=(X=8,Y=8,Z=-8), B=(X=0,Y=0,Z=0)) {template="Add::Execute(in A,in B,out Result)"}
   n14 = &RigVMFunction_MathVectorSub(
    A=@n11.Transform.Translation
    B=&RigVMFunction_MathTransformTransformVector(
     Transform=&RigUnit_GetBoneTransform(Bone=dyn_head_driver_tubeA_l).Transform
     Location=&RigVMFunction_MathVectorSelectBool(Condition=False, IfFalse=(X=0,Y=0,Z=0), IfTrue=@n13.Result).Result
    ).Result
   ) {template="Subtract::Execute(in A,in B,out Result)"}
   > &RigUnit_SetBoneRotation(
    Bone=dyn_head_driver_tubeA_l
    Space=GlobalSpace
    Weight=1
    bPropagateToChildren=True
    Rotation=&RigVMFunction_MathQuaternionMul(
     B=@n11.Transform.Rotation
     A=&RigVMFunction_MathQuaternionFromTwoVectors(A=@n12.Result, B=@n14.Result) {template="FromTwoVectors::Execute(in A,in B,out Result)"}.Result
    ) {template="Multiply::Execute(in A,in B,out Result)"}.Result
   )
   > &RigUnit_DebugTransformMutable(
    Color=(R=1,G=0,B=0,A=1)
    DebugDrawSettings=(DepthPriority=SDPG_Foreground,Lifetime=-1)
    Mode=Axes
    Scale=20
    Space=None
    Thickness=0.3
    WorldOffset=(Rotation=(X=0,Y=0,Z=0,W=1),Translation=(X=0,Y=0,Z=0),Scale3D=(X=1,Y=1,Z=1))
    bEnabled=false
    Transform=&RigUnit_GetBoneTransform(Bone=dyn_head_driver_tubeA_l).Transform
   )
   > &RigUnit_DebugTransformMutable(
    Color=(R=1,G=0,B=0,A=1)
    DebugDrawSettings=(DepthPriority=SDPG_Foreground,Lifetime=-1)
    Mode=Axes
    Scale=20
    Space=dyn_tubeD_01_l
    Thickness=0
    Transform=(Rotation=(X=0,Y=0,Z=0,W=1),Translation=(X=0,Y=0,Z=0),Scale3D=(X=1,Y=1,Z=1))
    WorldOffset=(Rotation=(X=0,Y=0,Z=0,W=1),Translation=(X=0,Y=0,Z=0),Scale3D=(X=1,Y=1,Z=1))
    bEnabled=false
   )
   n15 = &RigUnit_GetBoneTransform(Bone=dyn_tubeA_03_l)
   n16 = &RigUnit_GetBoneTransform(Bone=dyn_driver_tubeD_l)
   n17 = &RigVMFunction_MathVectorSelectBool(
    Condition=False
    IfFalse=(X=0,Y=0,Z=0)
    IfTrue=&RigVMFunction_MathVectorAdd(A=(X=8,Y=8,Z=-8), B=(X=0,Y=0,Z=0)) {template="Add::Execute(in A,in B,out Result)"}.Result
   )
   n18 = &RigVMFunction_MathQuaternionFromTwoVectors(
    A=&RigVMFunction_MathQuaternionGetAxis(Axis=X, Quaternion=@n15.Transform.Rotation) {template="Axis::Execute(in Quaternion,in Axis,out Result)"}.Result
    B=&RigVMFunction_MathVectorSub(
     A=@n15.Transform.Translation
     B=&RigVMFunction_MathTransformTransformVector(Transform=@n16.Transform, Location=@n17.Result).Result
    ) {template="Subtract::Execute(in A,in B,out Result)"}.Result
   ) {template="FromTwoVectors::Execute(in A,in B,out Result)"}
   > &RigUnit_SetBoneRotation(
    Bone=dyn_driver_tubeD_l
    Space=GlobalSpace
    Weight=1
    bPropagateToChildren=True
    Rotation=&RigVMFunction_MathQuaternionSlerp(
     T=0
     B=&RigUnit_GetBoneTransform(Bone=dyn_driver_tubeD_l).Transform.Rotation
     A=&RigVMFunction_MathQuaternionMul(B=@n15.Transform.Rotation, A=@n18.Result) {template="Multiply::Execute(in A,in B,out Result)"}.Result
    ) {template="Interpolate::Execute(in A,in B,in T,out Result)"}.Result
   )
   > &RigUnit_DebugTransformMutable(
    Color=(R=1,G=0,B=0,A=1)
    DebugDrawSettings=(DepthPriority=SDPG_Foreground,Lifetime=-1)
    Mode=Axes
    Scale=20
    Space=None
    Thickness=0.3
    WorldOffset=(Rotation=(X=0,Y=0,Z=0,W=1),Translation=(X=0,Y=0,Z=0),Scale3D=(X=1,Y=1,Z=1))
    bEnabled=false
    Transform=&RigUnit_GetBoneTransform(Bone=dyn_driver_tubeD_l).Transform
   )
   > &RigUnit_DebugTransformMutable(
    Color=(R=1,G=0,B=0,A=1)
    DebugDrawSettings=(DepthPriority=SDPG_Foreground,Lifetime=-1)
    Mode=Axes
    Scale=30
    Space=dyn_tubeA_03_l
    Thickness=0.5
    Transform=(Rotation=(X=0,Y=0,Z=0,W=1),Translation=(X=0,Y=0,Z=0),Scale3D=(X=1,Y=1,Z=1))
    WorldOffset=(Rotation=(X=0,Y=0,Z=0,W=1),Translation=(X=0,Y=0,Z=0),Scale3D=(X=1,Y=1,Z=1))
    bEnabled=false
   )
   n19 = &RigUnit_GetBoneTransform(Bone=dyn_tubeC_03_r)
   n20 = &RigVMFunction_MathQuaternionGetAxis(Axis=X, Quaternion=@n19.Transform.Rotation) {template="Axis::Execute(in Quaternion,in Axis,out Result)"}
   n21 = &RigVMFunction_MathVectorAdd(A=(X=8,Y=8,Z=-8), B=(X=0,Y=0,Z=0)) {template="Add::Execute(in A,in B,out Result)"}
   n22 = &RigVMFunction_MathVectorSub(
    B=@n19.Transform.Translation
    A=&RigVMFunction_MathTransformTransformVector(
     Transform=&RigUnit_GetBoneTransform(Bone=dyn_driver_tubeB_r).Transform
     Location=&RigVMFunction_MathVectorSelectBool(Condition=False, IfFalse=(X=0,Y=0,Z=0), IfTrue=@n21.Result).Result
    ).Result
   ) {template="Subtract::Execute(in A,in B,out Result)"}
   > &RigUnit_SetBoneRotation(
    Bone=dyn_driver_tubeB_r
    Space=GlobalSpace
    Weight=1
    bPropagateToChildren=True
    Rotation=&RigVMFunction_MathQuaternionMul(
     B=@n19.Transform.Rotation
     A=&RigVMFunction_MathQuaternionFromTwoVectors(A=@n20.Result, B=@n22.Result) {template="FromTwoVectors::Execute(in A,in B,out Result)"}.Result
    ) {template="Multiply::Execute(in A,in B,out Result)"}.Result
   )
   > &RigUnit_DebugTransformMutable(
    Color=(R=1,G=0,B=0,A=1)
    DebugDrawSettings=(DepthPriority=SDPG_Foreground,Lifetime=-1)
    Mode=Axes
    Scale=20
    Space=dyn_head_driver_tubeA_r
    Thickness=0
    Transform=(Rotation=(X=0,Y=0,Z=0,W=1),Translation=(X=0,Y=0,Z=0),Scale3D=(X=1,Y=1,Z=1))
    WorldOffset=(Rotation=(X=0,Y=0,Z=0,W=1),Translation=(X=0,Y=0,Z=0),Scale3D=(X=1,Y=1,Z=1))
    bEnabled=false
   )
   n23 = &RigUnit_GetBoneTransform(Bone=dyn_tubeM_01_r)
   n24 = &RigVMFunction_MathQuaternionGetAxis(Axis=X, Quaternion=@n23.Transform.Rotation) {template="Axis::Execute(in Quaternion,in Axis,out Result)"}
   n25 = &RigVMFunction_MathVectorAdd(A=(X=8,Y=8,Z=-8), B=(X=0,Y=0,Z=0)) {template="Add::Execute(in A,in B,out Result)"}
   n26 = &RigVMFunction_MathVectorSub(
    B=@n23.Transform.Translation
    A=&RigVMFunction_MathTransformTransformVector(
     Transform=&RigUnit_GetBoneTransform(Bone=dyn_head_driver_tubeA_r).Transform
     Location=&RigVMFunction_MathVectorSelectBool(Condition=False, IfFalse=(X=0,Y=0,Z=0), IfTrue=@n25.Result).Result
    ).Result
   ) {template="Subtract::Execute(in A,in B,out Result)"}
   > &RigUnit_SetBoneRotation(
    Bone=dyn_head_driver_tubeA_r
    Space=GlobalSpace
    Weight=1
    bPropagateToChildren=True
    Rotation=&RigVMFunction_MathQuaternionMul(
     B=@n23.Transform.Rotation
     A=&RigVMFunction_MathQuaternionFromTwoVectors(A=@n24.Result, B=@n26.Result) {template="FromTwoVectors::Execute(in A,in B,out Result)"}.Result
    ) {template="Multiply::Execute(in A,in B,out Result)"}.Result
   )
   > &RigUnit_DebugTransformMutable(
    Color=(R=1,G=0,B=0,A=1)
    DebugDrawSettings=(DepthPriority=SDPG_Foreground,Lifetime=-1)
    Mode=Axes
    Scale=20
    Space=None
    Thickness=0.3
    WorldOffset=(Rotation=(X=0,Y=0,Z=0,W=1),Translation=(X=0,Y=0,Z=0),Scale3D=(X=1,Y=1,Z=1))
    bEnabled=false
    Transform=&RigUnit_GetBoneTransform(Bone=dyn_head_driver_tubeA_r).Transform
   )
   > &RigUnit_DebugTransformMutable(
    Color=(R=1,G=0,B=0,A=1)
    DebugDrawSettings=(DepthPriority=SDPG_Foreground,Lifetime=-1)
    Mode=Axes
    Scale=30
    Space=dyn_tubeM_01_r
    Thickness=0.5
    Transform=(Rotation=(X=0,Y=0,Z=0,W=1),Translation=(X=0,Y=0,Z=0),Scale3D=(X=1,Y=1,Z=1))
    WorldOffset=(Rotation=(X=0,Y=0,Z=0,W=1),Translation=(X=0,Y=0,Z=0),Scale3D=(X=1,Y=1,Z=1))
    bEnabled=false
   )
   > &RigUnit_DebugTransformMutable(
    Color=(R=1,G=0,B=0,A=1)
    DebugDrawSettings=(DepthPriority=SDPG_Foreground,Lifetime=-1)
    Mode=Axes
    Scale=20
    Space=dyn_tubeA_03_r
    Thickness=1
    Transform=(Rotation=(X=0,Y=0,Z=0,W=1),Translation=(X=0,Y=0,Z=0),Scale3D=(X=1,Y=1,Z=1))
    WorldOffset=(Rotation=(X=0,Y=0,Z=0,W=1),Translation=(X=0,Y=0,Z=0),Scale3D=(X=1,Y=1,Z=1))
    bEnabled=false
   )
   n27 = &RigUnit_GetBoneTransform(Bone=dyn_tubeA_03_r)
   n28 = &RigVMFunction_MathQuaternionGetAxis(Axis=X, Quaternion=@n27.Transform.Rotation) {template="Axis::Execute(in Quaternion,in Axis,out Result)"}
   n29 = &RigVMFunction_MathVectorAdd(A=(X=8,Y=8,Z=-8), B=(X=0,Y=0,Z=0)) {template="Add::Execute(in A,in B,out Result)"}
   n30 = &RigVMFunction_MathVectorSub(
    B=@n27.Transform.Translation
    A=&RigVMFunction_MathTransformTransformVector(
     Transform=&RigUnit_GetBoneTransform(Bone=dyn_driver_tubeD_r).Transform
     Location=&RigVMFunction_MathVectorSelectBool(Condition=False, IfFalse=(X=0,Y=0,Z=0), IfTrue=@n29.Result).Result
    ).Result
   ) {template="Subtract::Execute(in A,in B,out Result)"}
   > &RigUnit_SetBoneRotation(
    Bone=dyn_driver_tubeD_r
    Space=GlobalSpace
    Weight=1
    bPropagateToChildren=True
    Rotation=&RigVMFunction_MathQuaternionMul(
     B=@n27.Transform.Rotation
     A=&RigVMFunction_MathQuaternionFromTwoVectors(A=@n28.Result, B=@n30.Result) {template="FromTwoVectors::Execute(in A,in B,out Result)"}.Result
    ) {template="Multiply::Execute(in A,in B,out Result)"}.Result
   )
   > &RigUnit_DebugTransformMutable(
    Color=(R=1,G=0,B=0,A=1)
    DebugDrawSettings=(DepthPriority=SDPG_Foreground,Lifetime=-1)
    Mode=Axes
    Scale=20
    Space=None
    Thickness=0.3
    WorldOffset=(Rotation=(X=0,Y=0,Z=0,W=1),Translation=(X=0,Y=0,Z=0),Scale3D=(X=1,Y=1,Z=1))
    bEnabled=false
    Transform=&RigUnit_GetBoneTransform(Bone=dyn_driver_tubeD_r).Transform
   )
   > &RigUnit_DebugTransformMutable(
    Color=(R=1,G=0,B=0,A=1)
    DebugDrawSettings=(DepthPriority=SDPG_Foreground,Lifetime=-1)
    Mode=Axes
    Scale=30
    Space=dyn_tubeA_03_r
    Thickness=0.5
    Transform=(Rotation=(X=0,Y=0,Z=0,W=1),Translation=(X=0,Y=0,Z=0),Scale3D=(X=1,Y=1,Z=1))
    WorldOffset=(Rotation=(X=0,Y=0,Z=0,W=1),Translation=(X=0,Y=0,Z=0),Scale3D=(X=1,Y=1,Z=1))
    bEnabled=false
   )
   n31 = &RigUnit_GetBoneTransform(Bone=neck_01)
   n32 = &RigVMFunction_MathVectorAdd(B=(X=0,Y=-5,Z=0), A=@n31.Transform.Translation) {template="Add::Execute(in A,in B,out Result)"}
   n33 = &RigVMFunction_MathTransformMakeRelative(
    Parent=(Rotation=(X=0,Y=0,Z=0,W=1),Translation=(X=0,Y=0,Z=0),Scale3D=(X=1,Y=1,Z=1))
    Parent.Rotation=@n31.Transform.Rotation
    Global=&RigUnit_GetBoneTransform(Bone=dyn_driver_tubeA_l).Transform
    Parent.Translation=@n32.Result
   )
   n34 = /Script/RigVM.RigVMFunction_MathVectorMul(B=(X=0,Y=1,Z=1), A=@n33.Local.Translation) {is=struct RigVMUnitNode, template="Multiply::Execute(in A,in B,out Result)"}
   n35 = /Script/RigVM.RigVMFunction_MathFloatClamp(
    Maximum=31
    Minimum=30
    Value=/Script/RigVM.RigVMFunction_MathVectorLength(Value=@n34.Result) {is=struct RigVMUnitNode, template="Length::Execute(in Value,out Result)"}.Result
   ) {is=struct RigVMUnitNode, template="Clamp::Execute(in Value,in Minimum,in Maximum,out Result)"}
   n36 = /Script/RigVM.RigVMFunction_MathVectorUnit(Value=@n34.Result) {is=struct RigVMUnitNode, template="Unit::Execute(in Value,out Result)"}
   n37 = &RigVMFunction_MathTransformMul(
    A=(Rotation=(X=0,Y=0,Z=0,W=1),Translation=(X=0,Y=0,Z=0),Scale3D=(X=1,Y=1,Z=1))
    B=(Rotation=(X=0,Y=0,Z=0,W=1),Translation=(X=0,Y=0,Z=0),Scale3D=(X=1,Y=1,Z=1))
    B.Rotation=@n31.Transform.Rotation
    A.Rotation=@n33.Local.Rotation
    A.Translation=&RigVMFunction_MathVectorAdd(
     A=&RigVMFunction_MathVectorMul(B=(X=1,Y=0,Z=0), A=@n33.Local.Translation) {template="Multiply::Execute(in A,in B,out Result)"}.Result
     B=/Script/RigVM.RigVMFunction_MathVectorScale(Factor=@n35.Result, Value=@n36.Result) {is=struct RigVMUnitNode, template="Scale::Execute(in Value,in Factor,out Result)"}.Result
    ) {template="Add::Execute(in A,in B,out Result)"}.Result
    B.Translation=@n32.Result
   ) {template="Multiply::Execute(in A,in B,out Result)"}
   /Script/ControlRig.RigUnit_SetBoneTransform(Bone=dyn_driver_tubeA_l, ExecutePin=(), Space=GlobalSpace, Weight=1, bPropagateToChildren=True, Transform=@n37.Result) {is=struct RigVMUnitNode}
   > &RigUnit_DebugTransformMutable(
    Color=(R=1,G=0,B=0,A=1)
    DebugDrawSettings=(DepthPriority=SDPG_Foreground,Lifetime=-1)
    Mode=Axes
    Scale=8
    Space=neck_01
    Thickness=0.4
    Transform=(Rotation=(X=0,Y=0,Z=0,W=1),Translation=(X=0,Y=0,Z=3),Scale3D=(X=1,Y=1,Z=1))
    WorldOffset=(Rotation=(X=0,Y=0,Z=0,W=1),Translation=(X=0,Y=0,Z=0),Scale3D=(X=1,Y=1,Z=1))
    bEnabled=True
   )
   > /Script/ControlRig.RigUnit_DebugArc(
    Color=(R=1,G=0,B=0,A=1)
    DebugDrawSettings=(DepthPriority=SDPG_Foreground,Lifetime=-1)
    Detail=16
    MaximumDegrees=360
    MinimumDegrees=0
    Radius=11
    Space=neck_01
    Thickness=0
    Transform=(Rotation=(X=0,Y=0,Z=0,W=1),Translation=(X=0,Y=-5,Z=0),Scale3D=(X=1,Y=1,Z=1))
    WorldOffset=(Rotation=(X=0,Y=0,Z=0,W=1),Translation=(X=0,Y=0,Z=0),Scale3D=(X=1,Y=1,Z=1))
    bEnabled=True
    Transform.Rotation=&RigVMFunction_MathQuaternionFromEuler(Euler=(X=0,Y=90,Z=0)).Result
   ) {is=struct RigVMUnitNode}
   &RigUnit_BeginExecution()
   &RigUnit_DebugTransformMutable(
    Color=(R=1,G=0,B=0,A=1)
    DebugDrawSettings=(DepthPriority=SDPG_Foreground,Lifetime=-1)
    ExecutePin=()
    Mode=Axes
    Scale=55
    Space=dyn_driver_tubeB_l
    Thickness=0
    Transform=(Rotation=(X=0,Y=0,Z=0,W=1),Translation=(X=0,Y=0,Z=0),Scale3D=(X=1,Y=1,Z=1))
    WorldOffset=(Rotation=(X=0,Y=0,Z=0,W=1),Translation=(X=0,Y=0,Z=0),Scale3D=(X=1,Y=1,Z=1))
    bEnabled=false
   )
   &RigVMFunction_MathQuaternionFromEuler(Euler=(X=13,Y=0,Z=0))
   /Script/RigVM.RigVMFunction_MathQuaternionScaleV2(Factor=0.8, Value=(X=0,Y=0,Z=0,W=1)) {is=struct RigVMUnitNode, template="Scale::Execute(in Value,in Factor,out Result)"}
   &RigVMFunction_MathQuaternionScaleV2(Factor=0.3, Value=(X=0,Y=0,Z=0,W=1)) {template="Scale::Execute(in Value,in Factor,out Result)"}
   &RigUnit_SetBoneRotation(Bone=dyn_driver_tubeA_l, ExecutePin=(), Rotation=(X=0,Y=0,Z=0,W=1), Space=GlobalSpace, Weight=1, bPropagateToChildren=True)
  ]}
]
hierarchy [
 root {loc=<0 0 0>, rot=<0 0 0 1>}
 pelvis {parent=root, loc=<0 -0.7321 92.4218>, rot=<0.0127 0.707 -0.0127 -0.707>}
 spine_01 {parent=pelvis, loc=<5.6002 -0.5128 0>, rot=<0 0 0.034 -0.9994>}
 spine_02 {parent=spine_01, loc=<8.0753 0.0139 0>, rot=<0 0 0.0238 -0.9997>}
 spine_03 {parent=spine_02, loc=<7.0393 0 0>, rot=<0 0 0.0149 -0.9999>}
 spine_04 {parent=spine_03, loc=<8.1024 0 0>, rot=<0 0 -0.0132 -0.9999>}
 spine_05 {parent=spine_04, loc=<10.0146 0 0>, rot=<0 0 -0.0015 -1>}
 clavicle_l {parent=spine_05, loc=<2.5856 -0.2866 -4.3777>, rot=<0.0465 -0.7289 -0.009 -0.6829>}
 clavicle_r {parent=spine_05, loc=<2.5856 -0.2865 4.3777>, rot=<0.7289 0.0465 -0.6829 0.009>}
 neck_01 {parent=spine_05, loc=<10.815 1.9196 0>, rot=<0 0 -0.1864 -0.9825>}
 neck_02 {parent=neck_01, loc=<4.7347 0.3045 0>, rot=<0 0 0.0363 -0.9993>}
 head {parent=neck_02, loc=<5.7632 0.5552 0>, rot=<0 0 0.115 -0.9934>}
 head_offset {parent=neck_02, loc=<5.7632 0.5552 0>, rot=<0 0 0.115 -0.9934>}
 dyn_parent_tubeC_l {parent=head_offset, loc=<13.6847 6.2252 -4.6696>, rot=<-0.0062 -0.707 -0.012 -0.7071>}
 dyn_driver_tubeC_l {parent=dyn_parent_tubeC_l, loc=<0 0 0>, rot=<0.6482 -0.4515 0.4818 -0.3793>}
 dyn_tubeC_01_l {parent=dyn_driver_tubeC_l, loc=<0 0 0>, rot=<0 0 0 1>}
 dyn_tubeC_02_l {parent=dyn_tubeC_01_l, loc=<17.3992 0 0>, rot=<0 0 0 1>}
 dyn_tubeC_03_l {parent=dyn_tubeC_02_l, loc=<5.3917 0.024 -2.3607>, rot=<0.8101 -0.5807 -0.08 -0.0084>}
 dyn_parent_tubeC_r {parent=head_offset, loc=<13.6847 6.2252 4.67>, rot=<-0.0062 -0.707 -0.012 -0.7071>}
 dyn_driver_tubeC_r {parent=dyn_parent_tubeC_r, loc=<-0.3197 -0.2401 -0.0018>, rot=<-0.3793 -0.4818 -0.4515 -0.6482>}
 dyn_tubeC_01_r {parent=dyn_driver_tubeC_r, loc=<0 0 0>, rot=<0 0 0 -1>}
 dyn_tubeC_02_r {parent=dyn_tubeC_01_r, loc=<-17.3991 -0.0004 -0.0001>, rot=<0 0 0 -1>}
 dyn_tubeC_03_r {parent=dyn_tubeC_02_r, loc=<-5.2083 0.2459 1.5735>, rot=<0.808 -0.5849 -0.065 -0.0288>}
 dyn_parent_tubeM_l {parent=head_offset, loc=<-0.2695 4.5431 -7.3068>, rot=<0.0029 -0.7071 -0.0029 -0.7071>}
 dyn_tubeM_01_l {parent=dyn_parent_tubeM_l, loc=<3.8597 0 0>, rot=<-0.3327 0.066 -0.6853 -0.6444>}
 dyn_parent_tubeM_r {parent=head_offset, loc=<-0.2695 4.5431 7.307>, rot=<0.0029 -0.7071 -0.0029 -0.7071>}
 dyn_tubeM_01_r {parent=dyn_parent_tubeM_r, loc=<-3.8595 0 -0.0003>, rot=<0.6444 -0.6853 -0.066 -0.3327>}
 dyn_parent_tubeA_r {parent=spine_05, loc=<12.9624 -2.1435 10.4084>, rot=<0.0284 -0.7065 -0.0284 -0.7065>}
 dyn_head_driver_tubeA_r {parent=dyn_parent_tubeA_r, loc=<0.0004 0 0.0001>, rot=<0.6444 -0.6853 -0.066 -0.3327>}
 dyn_driver_tubeA_r {parent=dyn_head_driver_tubeA_r, loc=<0 0.0003 -0.0001>, rot=<0.0184 -0.8714 0.1023 -0.4795>}
 dyn_tubeA_01_r {parent=dyn_driver_tubeA_r, loc=<0 0 0>, rot=<0 0 0 1>}
 dyn_tubeA_02_r {parent=dyn_tubeA_01_r, loc=<11.8137 0 0>, rot=<0 0 0 1>}
 dyn_tubeA_03_r {parent=dyn_tubeA_02_r, loc=<7.8016 6.3703 9.1547>, rot=<0.0957 -0.8435 -0.5134 0.1254>}
 dyn_parent_tubeA_l {parent=spine_05, loc=<12.962 -2.1435 -10.408>, rot=<0.0284 -0.7065 -0.0284 -0.7065>}
 dyn_head_driver_tubeA_l {parent=dyn_parent_tubeA_l, loc=<0 0 0>, rot=<-0.3327 0.066 -0.6853 -0.6444>}
 dyn_driver_tubeA_l {parent=dyn_head_driver_tubeA_l, loc=<0 0 0>, rot=<0.1023 0.4795 -0.0183 -0.8714>}
 dyn_tubeA_01_l {parent=dyn_driver_tubeA_l, loc=<0 0 0>, rot=<0 0 0 -1>}
 dyn_tubeA_02_l {parent=dyn_tubeA_01_l, loc=<11.8137 0 0>, rot=<0 0 0 -1>}
 dyn_tubeA_03_l {parent=dyn_tubeA_02_l, loc=<7.802 -6.37 9.155>, rot=<0.5134 -0.1254 0.0957 -0.8435>}
 dyn_parent_tubeB_l {parent=spine_05, loc=<18.1607 -9.328 -3.127>, rot=<0.0284 -0.7065 -0.0284 -0.7065>}
 dyn_driver_tubeB_l {parent=dyn_parent_tubeB_l, loc=<0 0 0>, rot=<-0.0054 0.6663 0.0061 -0.7456>}
 dyn_tubeB_01_l {parent=dyn_driver_tubeB_l, loc=<0 0 0>, rot=<0 0 0 -1>}
 dyn_parent_tubeB_r {parent=spine_05, loc=<18.1607 -9.328 3.127>, rot=<0.0284 -0.7065 -0.0284 -0.7065>}
 dyn_driver_tubeB_r {parent=dyn_parent_tubeB_r, loc=<0 0 0.0002>, rot=<0.7456 0.0061 -0.6663 -0.0054>}
 dyn_tubeB_01_r {parent=dyn_driver_tubeB_r, loc=<0 0 0>, rot=<0 0 0 1>}
 dyn_parent_tubeD_l {parent=spine_05, loc=<15.5352 -13.2361 -11.7709>, rot=<0.0284 -0.7065 -0.0284 -0.7065>}
 dyn_driver_tubeD_l {parent=dyn_parent_tubeD_l, loc=<0 0 0>, rot=<-0.1689 0.4788 -0.1625 -0.846>}
 dyn_tubeD_01_l {parent=dyn_driver_tubeD_l, loc=<0 0 0>, rot=<0 0 0 -1>}
 dyn_parent_tubeD_r {parent=spine_05, loc=<15.5352 -13.2361 11.771>, rot=<0.0284 -0.7065 -0.0284 -0.7065>}
 dyn_driver_tubeD_r {parent=dyn_parent_tubeD_r, loc=<0.0001 0 0.0004>, rot=<0.846 -0.1625 -0.4788 -0.1689>}
 dyn_tubeD_01_r {parent=dyn_driver_tubeD_r, loc=<0 0 0>, rot=<0 0 0 1>}
]
