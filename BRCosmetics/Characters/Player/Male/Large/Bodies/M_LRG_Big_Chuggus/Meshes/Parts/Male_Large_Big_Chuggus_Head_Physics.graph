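# Physics asset for the head part of the Big Chuggus large male player cosmetic, set up for physical motion of four tubes, A and C on each side.
# Each tube has one capsule body on its first dynamic bone, constrained to a tiny capsule on a matching driver bone.
# The four constraints are independent pairs with moderate swing limits, wider on the A tubes, and tight twist limits.
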PhysicsAsset /BRCosmetics/Characters/Player/Male/Large/Bodies/M_LRG_Big_Chuggus/Meshes/Parts/Male_Large_Big_Chuggus_Head_Physics
bodies {
 dyn_driver_tubeA_l [{shapeType=Capsule, length=0.505, radius=0.505, rotation=<rot 0 -90 0>}]
 dyn_driver_tubeA_r [{shapeType=Capsule, length=0.505, radius=0.505, rotation=<rot 0 -90 0>}]
 dyn_driver_tubeC_l [{shapeType=Capsule, length=0.505, radius=0.505, rotation=<rot 0 -90 0>}]
 dyn_driver_tubeC_r [{shapeType=Capsule, length=0.505, radius=0.505, rotation=<rot 0 -90 0>}]
 dyn_tubeA_01_l [{shapeType=Capsule, length=4.25, radius=3, center=<1 0 0>, rotation=<rot 90 0 0>}]
 dyn_tubeA_01_r [{shapeType=Capsule, length=4.25, radius=3, center=<1 0 0>, rotation=<rot 90 0 0>}]
 dyn_tubeC_01_l [{shapeType=Capsule, length=6.5, radius=3.005, center=<0.8037 -0.064 0.0001>, rotation=<rot 90 -165.9637 -115.9655>}]
 dyn_tubeC_01_r [{shapeType=Capsule, length=7, radius=3.005, center=<-0.75 0 0>, rotation=<rot 90 90 -40.0002>}]
}
constraints [
 {bone1Name=dyn_tubeC_01_l, bone2Name=dyn_driver_tubeC_l, swing1LimitDegrees=45, swing2LimitDegrees=35, twistLimitDegrees=5}
 {bone1Name=dyn_tubeC_01_r, bone2Name=dyn_driver_tubeC_r, swing1LimitDegrees=45, swing2LimitDegrees=32, twistLimitDegrees=5}
 {bone1Name=dyn_tubeA_01_l, bone2Name=dyn_driver_tubeA_l, swing1LimitDegrees=45, swing2LimitDegrees=60, twistLimitDegrees=5}
 {bone1Name=dyn_tubeA_01_r, bone2Name=dyn_driver_tubeA_r, swing1LimitDegrees=45, swing2LimitDegrees=60, twistLimitDegrees=5}
]
