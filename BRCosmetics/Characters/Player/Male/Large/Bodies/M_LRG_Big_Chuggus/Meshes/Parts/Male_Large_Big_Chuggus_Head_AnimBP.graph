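# Animation Blueprint for the head part of the male large Big Chuggus character body, adding secondary physics motion and rig adjustments.
# It copies the pose from the attached parent mesh, then runs RigidBody simulation in root-bone space with an override physics asset, custom gravity and scaled component motion.
# The simulated pose feeds the head's Control Rig before output.
# A head-to-head_offset Copy Bone node is unconnected, and the event graph has no logic.
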AnimBlueprint /BRCosmetics/Characters/Player/Male/Large/Bodies/M_LRG_Big_Chuggus/Meshes/Parts/Male_Large_Big_Chuggus_Head_AnimBP
target_skeleton=&/Male_Large_Big_Chuggus_Head_Skeleton
graphs [
 AnimGraph {nodes [
   Copy Bone - Source Bone: head - Target Bone: head_offset!(
    ComponentPose:Component Space Pose Link Structure
    Alpha:float=0.65
    bCopyTranslation:bool=False
    bCopyRotation:bool=true
    bCopyScale:bool=False
   ) -> (Pose:Component Space Pose Link Structure) {
    is=title AnimGraphNode_CopyBone
    controlSpace=BCS_BoneSpace
    alphaBoolBlend {blendOption=Linear}
    sourceBone {boneName=head}
    targetBone {boneName=head_offset}
   }
   n1 = RigidBody(ComponentPose=Local To Component(LocalPose=Copy Pose From Mesh() {bUseAttachedParent=true})) {
    bClampLinearTranslationLimitToRefPose=true
    bOverrideWorldGravity=true
    overridePhysicsAsset=&/Male_Large_Big_Chuggus_Head_Physics
    simulationSpace=BaseBoneSpace
    alphaBoolBlend {blendOption=Linear}
    baseBoneRef {boneName=root}
    componentAppliedLinearAccClamp=<1200 1200 1200>
    componentLinearAccScale=<8 8 0.5>
    componentLinearVelScale=<15 15 2>
    overrideWorldGravity {z=-950}
   }
   n2 = Control Rig(Source=Component To Local(ComponentPose=@n1)) {
    alphaBoolBlend {blendOption=Linear}
    controlRigAssetReference {blueprintRigClass=&/Male_Large_Big_Chuggus_Head_CtrlRig.Male_Large_Big_Chuggus_Head_CtrlRig_C}
   }
   Output Pose(Result=@n2) {layerGroup=None}
  ]}
 EventGraph {nodes [
   TryGetPawnOwner()
   Event BlueprintUpdateAnimation()
  ]}
]
variables []
